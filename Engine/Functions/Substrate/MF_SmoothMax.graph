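# MF_SmoothMax is a Substrate engine material function that returns a smooth maximum of scalar inputs A and B.
# It negates both inputs, passes them to MF_SmoothMin, then negates that result, so all blending logic lives in the SmoothMin function.
# The scalar distance input sets the blend distance (k) and passes through unchanged.
# All three inputs use their preview values as defaults, so each is optional.
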MaterialFunction /Engine/Functions/Substrate/MF_SmoothMax
graph {nodes [
  n1 = Multiply(A=FunctionInput() {bUsePreviewValueAsDefault=true, inputName=A, inputType=FunctionInput_Scalar}) {constB=-1}
  n2 = FunctionInput() {bUsePreviewValueAsDefault=true, inputName=B, inputType=FunctionInput_Scalar, sortPriority=1, previewValue=<1 0 0 1>}
  n3 = Multiply(A=@n2) {constB=-1}
  n4 = FunctionInput() {
   bUsePreviewValueAsDefault=true
   description="blend distance of the two inputs (called k)"
   inputName=distance
   inputType=FunctionInput_Scalar
   sortPriority=2
   previewValue=<0.1 0 0 1>
  }
  FunctionOutput(None=Multiply(A=MaterialFunctionCall!(A=@n1, B=@n3, distance=@n4) {materialFunction=&/MF_SmoothMin}.SmoothMin) {constB=-1}) {outputName=SmoothMax}
 ]}
