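# Material function for the FirstClass character that draws stylized hatching lines into emissive, behind a static switch.
# Threshold terms and a remapped lighting-style input set a tone level, smooth-rounded to blend six hatching layers from channel-packed textures (UV or triplanar) or line math.
# A switch picks procedural or texture hatching, with added macro noise; line color is LUT/time-of-day adjusted or darkened.
# Also builds Voronoi-distorted lines, shading bands, terminator line and sun-vector fallback.
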MaterialFunction /Game/Characters/Player/Male/Medium/Bodies/M_MED_FirstClass/Materials/MF_FirstClass_HatchingLines
graph {nodes [
  n1 = NamedRerouteUsage()
  n2 = Min(A=@n1, B=NamedRerouteUsage())
  n3 = ScalarParameter() {group=01_HatchingLines, parameterName=ThresholdVector_Strenght, sliderMax=1}
  n4 = Add(
   A=Power(
    Base=LinearInterpolate(A=@n1, B=@n2, Alpha=@n3)
    Exp=ScalarParameter() {defaultValue=1, group=01_HatchingLines, parameterName=Threshold_Exponent, sliderMax=32}
   ) {constExponent=8}
   B=ScalarParameter() {group=01_HatchingLines, parameterName=Threshold_Offset, sliderMax=2, sliderMin=-2}
  ) {constB=0.8}
  n5 = Saturate(None=MaterialFunctionCall(
    Input=NamedRerouteUsage()
    Input Low=ScalarParameter() {defaultValue=0.2, group=01_HatchingLines, parameterName=NdotL_Min, sliderMax=1}
    Input High=ScalarParameter() {defaultValue=0.7, group=01_HatchingLines, parameterName=NdotL_Max, sliderMax=1}
   ) {materialFunction=&RemapValue_0-1}.Result)
  n6 = LinearInterpolate(
   A=ScalarParameter() {defaultValue=6, group=01_HatchingLines, parameterName=Lerp_Max, sliderMax=16}
   B=ScalarParameter() {group=01_HatchingLines, parameterName=Lerp_Min, sliderMax=8, sliderMin=-4}
   Alpha=Saturate(None=LinearInterpolate(A=@n4, Alpha=@n5))
  ) {constA=6, constB=0}
  n7 = Constant() {r=0.5}
  Composite!(In=@n6, RampWidth=@n7) {subgraphName=SmoothFloor}
  Composite(In=@n6, RampWidth=@n7) {subgraphName=SmoothRound}
  n8 = ScalarParameter() {defaultValue=128, group=01_HatchingLines, parameterName=Line_Amount, sliderMax=256}
  n9 = Multiply(
   A=Rotator(
    Coordinate=TextureCoordinate()
    Time=ScalarParameter() {defaultValue=0.2, group=01_HatchingLines, parameterName=Line_Rot, sliderMax=1, sliderMin=-1}
   ) {speed=3.1415}
   B=@n8
  ) {constB=64}
  n10 = Floor(None=@n9)
  Divide(A=@n10, B=@n8)
  n11 = NamedRerouteUsage()
  n12 = VectorParameter() {group=01_HatchingLines, parameterName=HatchingLine_Color, sortPriority=2, defaultValue=<rgba 0.02 0.02 0.02 0>}
  n13 = MaterialFunctionCall!(In=@n12.RGB, LUT, LUT Multiplier) {materialFunction=/Game/Packages/Fortress_SharedMaterials/Anime_Master_Material/MF_ApplyLUT}
  n14 = NamedRerouteUsage()
  n15 = ScalarParameter() {group=01_HatchingLines, parameterName=Lines_DiffuseColorDarkening, sliderMax=1}
  n16 = NamedRerouteUsage()
  n17 = LinearInterpolate(
   A=StaticSwitchParameter(
    True=MaterialFunctionCall!(Color=@n13.Result) {materialFunction=/Game/Packages/Fortress_SharedMaterials/Anime_Master_Material/MF_TODColorAdjustment}.Result
    False=Multiply(A=@n14, B=@n15)
   ) {group=01_HatchingLines, parameterName=UseHatchingColor, sortPriority=1}
   B=NamedRerouteUsage()
   Alpha=Max(A=NamedRerouteUsage(), B=OneMinus(None=@n16))
  )
  FunctionOutput(None=StaticSwitchParameter(
    True=SetMaterialAttributes!(MaterialAttributes=@n11, EmissiveColor=@n17) {attributeSetTypes [#1]}
    False=NamedRerouteUsage()
   ) {defaultValue=true, group=01_HatchingLines, parameterName=UseHatchingLines, sortPriority=0})
  n18 = ScalarParameter() {defaultValue=0.5, group=01_HatchingLines, parameterName=Hatching_Position}
  n19 = Add(A=@n18, B=ScalarParameter() {defaultValue=0.05, group=01_HatchingLines, parameterName=TransitionThickness, sliderMax=1})
  n20 = NamedRerouteUsage()
  n21 = Normalize(VectorInput=NamedRerouteUsage())
  n22 = MaterialFunctionCall(Float3=Power(Base=Abs(None=@n21)) {constExponent=4}) {materialFunction=&BreakOutFloat3Components}
  n23 = Add(A=@n22.R, B=@n22.R)
  n24 = DotProduct(A=@n21, B=Constant3Vector() {constant=<rgba 0 0 -1 0>})
  LinearInterpolate(
   A=LinearInterpolate(A=NamedRerouteUsage(), Alpha=SmoothStep(Min=@n18, Max=@n19, Value=@n20) {constMax=0.85, constMin=0.75}) {constA=1}
   Alpha=Saturate(None=Add(A=@n23, B=@n24))
  )
  n25 = NamedRerouteUsage()
  n26 = DotProduct(A=VertexInterpolator(VS=PreSkinnedNormal()).PS, B=Constant3Vector() {constant=<rgba 0 0 1 0>})
  n27 = Power(Base=Add(A=Multiply(A=@n26) {constB=0.5}) {constB=0.5}) {constExponent=1}
  LinearInterpolate(Alpha=Saturate(None=Multiply(A=@n25, B=@n27)))
  MaterialFunctionCall!(A, B, distance) {materialFunction=/Engine/Functions/Substrate/MF_SmoothMax1}
  MaterialFunctionCall!(In=@n6, Ramp Width=@n7) {materialFunction=/Engine/Functions/Engine_MaterialFunctions01/Math/SmoothCeil}
  MaterialFunctionCall!(A, B, distance) {materialFunction=/Engine/Functions/Substrate/MF_SmoothMax}
  MaterialFunctionCall!(A, B, distance) {materialFunction=/Engine/Functions/Substrate/MF_SmoothMin}
  n28 = NamedRerouteUsage()
  n29 = ScalarParameter() {defaultValue=1, group=01_HatchingLines, parameterName=NdotL_Offset, sliderMax=1, sliderMin=-1}
  Multiply(A=Saturate(None=Add(A=@n28, B=@n29)))
  n30 = NamedRerouteDeclaration(None=DotProduct(A=NamedRerouteUsage(), B=PixelNormalWS())) {name=NdotL, nodeColor=<rgba 0 0.3203 1 1>}
  NamedRerouteDeclaration(None=Add(A=Multiply(A=@n30) {constB=0.5}) {constB=0.5}) {name=NdotL01, nodeColor=<rgba 1 0.3984 0 1>}
  n31 = PreSkinnedPosition()
  NamedRerouteDeclaration(None=ShaderStageSwitch(PixelShader=VertexInterpolator(VS=@n31).PS, VertexShader=@n31)) {name=PSL_P, nodeColor=<rgba 1 0.3739 0 1>}
  n32 = ScalarParameter() {defaultValue=6, group=01_HatchingLines, parameterName=ProjectionTexture1_Size}
  n33 = NamedRerouteUsage()
  n34 = NamedRerouteUsage()
  n35 = Constant() {r=0.125}
  n36 = Constant() {r=0.5}
  n37 = TextureObject() {samplerType=SAMPLERTYPE_Masks, texture=/Game/Characters/Player/Male/Medium/Bodies/M_MED_FirstClass/Textures/T_HatchingPattern04_FX}
  n38 = TextureCoordinate()
  n39 = Rotator(
   Coordinate=Multiply(A=@n38, B=ScalarParameter() {defaultValue=12, group=01_HatchingLines, parameterName=HatchingTextture1_Tiling, sliderMax=24})
   Time=ScalarParameter() {defaultValue=0.25, group=01_HatchingLines, parameterName=HatchingTextture1_Rotation, sliderMax=1, sliderMin=-1}
  ) {speed=3.1415}
  n40 = TextureObject() {samplerType=SAMPLERTYPE_Masks, texture=&T_HatchingPattern04_FX}
  n41 = StaticBoolParameter() {group=01_HatchingLines, parameterName=UseTriPlanarProjectionTextureHatchingLines, sortPriority=1}
  n42 = MaterialFunctionCall(Float3=StaticSwitch(
    True=MaterialFunctionCall(
     TextureObject=@n37
     TextureSize=@n32
     WorldPosition=@n33
     World Space Normal=@n34
     RotZAxis=@n35
     RotXAxis=@n35
     RotYAxis=@n35
     ProjectionTransitionContrast=@n36
    ) {materialFunction=/Game/Weapons/WeaponSkins/Wraps/MaterialFunctions/WorldAlignedTexture}.XYZ Texture
    False=TextureSample(UVs=@n39, Tex=@n40).RGB
    Value=@n41
   )) {materialFunction=&BreakOutFloat3Components}
  MaterialExpressionReroute_0 = Reroute(None=@n6)
  n43 = Frac(None=Add(A=@MaterialExpressionReroute_0.In) {constB=-0.5})
  MaterialExpressionReroute_1 = Reroute(None=@n7)
  n44 = OneMinus(None=@MaterialExpressionReroute_1.RampWidth)
  n45 = Max(A=Divide(A=Subtract(A=@n43, B=@n44), B=@MaterialExpressionReroute_1.RampWidth), B=Constant() {r=0.5}) {desc=It is good to make functions look like a lancer}
  n46 = Add(A=Round(None=@MaterialExpressionReroute_0.In), B=ConstantBiasScale(None=Cosine(None=@n45)))
  MaterialExpressionReroute_5 = Reroute(None=@n46)
  n47 = Add(A=@MaterialExpressionReroute_5.Output) {constB=-1}
  n48 = Saturate(None=@n47)
  n49 = Add(A=@n47) {constB=-1}
  n50 = Saturate(None=@n49)
  n51 = Add(A=Multiply(A=@n42.R, B=@n48), B=Multiply(A=@n42.G, B=@n50))
  n52 = Add(A=@n49) {constB=-1}
  n53 = Saturate(None=@n52)
  n54 = Multiply(A=@n42.B, B=@n53)
  n55 = ScalarParameter() {defaultValue=4, group=01_HatchingLines, parameterName=ProjectionTexture2_Size}
  n56 = Add(A=@n35, B=Constant() {r=0.25})
  n57 = Rotator(
   Coordinate=Multiply(A=@n38, B=ScalarParameter() {defaultValue=10, group=01_HatchingLines, parameterName=HatchingTextture2_Tiling, sliderMax=24})
   Time=ScalarParameter() {defaultValue=-0.25, group=01_HatchingLines, parameterName=HatchingTextture2_Rotation, sliderMax=1, sliderMin=-1}
  ) {speed=3.1415}
  n58 = MaterialFunctionCall(Float3=StaticSwitch(
    True=MaterialFunctionCall(
     TextureObject=@n37
     TextureSize=@n55
     WorldPosition=@n33
     World Space Normal=@n34
     RotZAxis=@n56
     RotXAxis=@n56
     RotYAxis=@n56
     ProjectionTransitionContrast=@n36
    ) {materialFunction=&WorldAlignedTexture}.XYZ Texture
    False=TextureSample(UVs=@n57, Tex=@n40).RGB
    Value=@n41
   )) {materialFunction=&BreakOutFloat3Components}
  n59 = Add(A=@n52) {constB=-1}
  n60 = Saturate(None=@n59)
  n61 = Add(A=@n59) {constB=-1}
  n62 = Saturate(None=@n61)
  n63 = Add(A=Add(A=Add(A=@n51, B=@n54), B=Multiply(A=@n58.R, B=@n60)), B=Multiply(A=@n58.G, B=@n62))
  n64 = Saturate(None=Add(A=@n61) {constB=-1})
  n65 = Multiply(A=@n58.B, B=@n64)
  NamedRerouteDeclaration(None=OneMinus(None=Add(A=@n63, B=@n65))) {name=ProjectedHatching, nodeColor=<rgba 1 0 0.0937 1>}
  n66 = PreSkinnedNormal()
  NamedRerouteDeclaration(None=ShaderStageSwitch(PixelShader=VertexInterpolator(VS=@n66).PS, VertexShader=@n66)) {name=PSL_N, nodeColor=<rgba 0 0.2666 1 1>}
  n67 = Multiply(A=NamedRerouteUsage(), B=ScalarParameter() {defaultValue=0.5, group=01_HatchingLines, parameterName=LineDistortion_RandScale, sliderMax=1})
  n68 = NamedRerouteUsage()
  n69 = ScalarParameter() {defaultValue=0.15, group=01_HatchingLines, parameterName=LineDistortion_RandRot, sliderMax=0.25}
  n70 = Rotator(
   Coordinate=Multiply(A=TextureCoordinate(), B=Add(A=@n67))
   Time=Add(A=ScalarParameter() {group=01_HatchingLines, parameterName=Lines_RotationOffset, sliderMax=1}, B=Multiply(A=@n68, B=@n69))
  ) {centerX=0, centerY=0, speed=-6.283}
  n71 = Frac(None=Multiply(A=ComponentMask(None=@n70) {r=true}, B=ScalarParameter() {defaultValue=180, group=01_HatchingLines, parameterName=UV_Lines_Amount}) {constB=32})
  n72 = Abs(None=Add(A=Multiply(A=@n71) {constB=2}) {constB=-1}) {desc=HatchingDF}
  n73 = NamedRerouteUsage()
  NamedRerouteDeclaration(None=Saturate(None=LinearInterpolate(B=@n72, Alpha=@n73) {constA=1})) {name=VornoiHatching, nodeColor=<rgba 0.9922 1 0 1>}
  n74 = MaterialFunctionCall(Float2=TextureCoordinate()) {materialFunction=&BreakOutFloat2Components}
  n75 = Multiply(A=@n74.G, B=Divide(B=ScalarParameter() {defaultValue=1, group=01_HatchingLines, parameterName=ShadowVariation_Ratio, sliderMax=2}))
  n76 = TextureSample(UVs=Multiply(
    A=MaterialFunctionCall(X=@n74.R, Y=@n75) {materialFunction=&MakeFloat2}.Result
    B=ScalarParameter() {defaultValue=2, group=01_HatchingLines, parameterName=ShadowVariation_Tiling, sliderMax=16}
   )) {samplerType=SAMPLERTYPE_LinearColor, texture=/Game/Global/Materials/Character_Parent_2023/Anime_Parent_2023/Textures/T_Hatching_Voronoi}
  NamedRerouteDeclaration(None=@n76.G) {name=VoronoiID, nodeColor=<rgba 1 0.2603 0.2703 1>}
  n77 = Abs(None=Add(A=Frac(None=@n9)) {constB=-0.5})
  n78 = OneMinus(None=SmoothStep(Max=ScalarParameter() {defaultValue=0.01, group=01_HatchingLines, parameterName=Line_Thickness, sliderMax=1}, Value=Power(Base=@n77)) {constMax=0.1})
  n79 = ComponentMask(None=@n78) {r=true}
  n80 = Constant() {r=4}
  n81 = OneMinus(None=Saturate(None=Fmod(A=@n10, B=@n80))) {desc=1}
  n82 = MaterialFunctionCall(Float2=@n81) {materialFunction=&BreakOutFloat2Components}
  n83 = Constant() {r=2}
  n84 = OneMinus(None=Saturate(None=Fmod(A=@n10, B=@n83))) {desc=1}
  n85 = MaterialFunctionCall(Float2=Subtract(A=@n84, B=@n81)) {materialFunction=&BreakOutFloat2Components}
  n86 = MaterialFunctionCall(Float2=Subtract(A=Constant() {r=1}, B=@n84)) {materialFunction=&BreakOutFloat2Components}
  n87 = MaterialFunctionCall(Float3=Multiply(
    A=@n79
    B=MaterialFunctionCall(X=@n82.R, Y=@n85.R, Z=@n86.R) {materialFunction=&MakeFloat3}.Result
   )) {materialFunction=&BreakOutFloat3Components}
  n88 = Add(A=Multiply(A=@n87.R, B=@n48), B=Multiply(A=@n87.G, B=@n50))
  n89 = Multiply(A=@n87.B, B=@n53)
  n90 = MaterialFunctionCall(Float3=Multiply(
    A=ComponentMask(None=@n78) {g=true}
    B=MaterialFunctionCall(X=@n82.G, Y=@n85.G, Z=@n86.G) {materialFunction=&MakeFloat3}.Result
   )) {materialFunction=&BreakOutFloat3Components}
  n91 = Add(A=Add(A=Add(A=@n88, B=@n89), B=Multiply(A=@n90.R, B=@n60)), B=Multiply(A=@n90.G, B=@n62))
  n92 = Multiply(A=@n90.B, B=@n64)
  NamedRerouteDeclaration(None=OneMinus(None=Add(A=@n91, B=@n92))) {name=LinearHatching, nodeColor=<rgba 1 0 0.0938 1>}
  NamedRerouteDeclaration(None=FunctionInput() {inputName=Color Mask, inputType=FunctionInput_Scalar, sortPriority=2}) {name=ColorMask, nodeColor=<rgba 1 0.3516 0 1>}
  n93 = Custom() {code="return View.DirectionalLightColor;", description=DirectionalLightColor, outputType=CMOT_Float4, inputs []}
  n94 = Custom() {code="return View.DirectionalLightDirection;", description=DirectionalLightDirection, inputs []}
  n95 = DotProduct(A=@n94, B=@n94)
  n96 = CollectionParameter() {
   collection=&FortniteMaterialParameters
   parameterName=SunlightDirectionalVector
  }
  n97 = ComponentMask(None=@n96) {b=true, g=true, r=true}
  NamedRerouteDeclaration(None=If(
    A=ComponentMask(None=@n93) {b=true, g=true, r=true}
    "A > B"=If(A=@n95, "A > B"=@n94, "A == B"=@n97, "A < B"=@n97) {desc=check if DirectionalLight exists}
    "A == B"=@n97
    "A < B"=@n97
   ) {desc=check if DirectionalLight color is Black}) {name=LightVector, nodeColor=<rgba 1 0 0.7031 1>}
  n98 = MaterialFunctionCall() {materialFunction=&CameraDirectionVector}
  n99 = Multiply(A=Constant3Vector() {constant=<rgba 1 1 0 0>}, B=@n98.Result)
  n100 = Add(A=Multiply(B=Normalize(VectorInput=@n99)) {constA=0.25}, B=@n98.Result)
  n101 = Multiply(A=@n98.Result, B=ScalarParameter() {defaultValue=10, group=01_HatchingLines, parameterName=ScreenPostionDepthBias, sliderMax=100}) {constB=50}
  n102 = NamedRerouteDeclaration(None=Add(A=ActorPositionWS(), B=VectorParameter() {group=01_HatchingLines, parameterName=VirtualPlanePivotOffset, defaultValue=<rgba 0 0 90 1>}.RGB)) {name=AdjustedActorPosition, nodeColor=<rgba 0.8281 0 1 1>}
  n103 = Constant() {r=64}
  n104 = Constant() {r=128}
  n105 = InverseLinearInterpolate(
   A=Constant() {r=45}
   B=Constant() {r=90}
   Value=MaterialFunctionCall() {materialFunction=/Engine/Functions/Engine_MaterialFunctions02/Utility/FOV}.FOV angle in degrees
  )
  NamedRerouteDeclaration(None=MaterialFunctionCall(
    Plane Normal Axis=Normalize(VectorInput=@n100)
    Plane Center=Add(A=@n101, B=@n102)
    UV Scale=LinearInterpolate(A=@n103, B=@n104, Alpha=@n105)
   ) {materialFunction=&VirtualPlaneCoordinates}.UVs) {name=ObjectAlighedScreenPosition, nodeColor=<rgba 1 0.867 0 1>}
  NamedRerouteDeclaration(None=ComponentMask(None=FunctionInput() {inputName=SSC, sortPriority=1}) {r=true}) {name=Threshold, nodeColor=<rgba 1 0 0.1406 1>}
  n106 = DotProduct(
   A=StaticSwitchParameter(True=NamedRerouteUsage(), False=PixelNormalWS()) {group=01_HatchingLines, parameterName=UsePreSkinnedLocalNormal4ThresholdMaskVector, sortPriority=0}
   B=Normalize(VectorInput=VectorParameter() {group=01_HatchingLines, parameterName=Threshold_Vector, defaultValue=<rgba 0 0 1 1>}.RGB)
  )
  NamedRerouteDeclaration(None=Add(A=Multiply(A=@n106) {constB=0.5}) {constB=0.5}) {name=ThresholdVector, nodeColor=<rgba 0.4937 0 1 1>}
  NamedRerouteDeclaration(None=SmoothStep(Value=@n76.R) {constMax=0.8, constMin=1}) {name=VoronoiIEdgeDistance, nodeColor=<rgba 1 0.2603 0.2703 1>}
  NamedRerouteDeclaration(None=@n76.B) {name=VoronoiICenterDistance, nodeColor=<rgba 1 0.2603 0.2703 1>}
  n107 = ScalarParameter() {defaultValue=3, group=01_HatchingLines, parameterName=ShadingCurve_Bands_Amount, sliderMax=8, sliderMin=2}
  n108 = Multiply(A=Saturate(None=NamedRerouteUsage()), B=@n107)
  n109 = Frac(None=@n108)
  n110 = Divide(
   A=SmoothStep(
    Min=ScalarParameter() {defaultValue=0.75, group=01_HatchingLines, parameterName=ShadingCurve_Min, sliderMax=1}
    Max=ScalarParameter() {defaultValue=1, group=01_HatchingLines, parameterName=ShadingCurve_Max, sliderMax=1}
    Value=@n109
   ) {constMin=0.5}
   B=@n107
  )
  n111 = Divide(A=Floor(None=@n108), B=@n107)
  NamedRerouteDeclaration(None=Multiply(A=Add(A=@n110, B=@n111)) {constB=3}) {name=Bands, nodeColor=<rgba 1 0.3984 0 1>}
  NamedRerouteDeclaration(None=Min(A=@n79, B=NamedRerouteUsage())) {name=HatchingMask, nodeColor=<rgba 0 0.9062 1 1>}
  n112 = ScalarParameter() {defaultValue=0.015, group=01_HatchingLines, parameterName=TerminatorLine_Thickness, sliderMax=0.2}
  n113 = ScalarParameter() {defaultValue=0.01, group=01_HatchingLines, parameterName=TerminatorLine_Transition, sliderMax=0.1}
  n114 = Add(
   A=NamedRerouteUsage()
   B=ScalarParameter() {defaultValue=-0.07, group=01_HatchingLines, parameterName=TerminatorLine_Bias, sliderMax=0.25, sliderMin=-0.25}
  )
  NamedRerouteDeclaration(None=SmoothStep(Min=Subtract(A=@n112, B=@n113), Max=@n112, Value=Abs(None=@n114)) {constMax=0.02, constMin=0.01}) {name=TerminatorLine, nodeColor=<rgba 0 1 0.625 1>}
  n115 = Constant()
  n116 = SetMaterialAttributes!(MaterialAttributes, EmissiveColor=Constant() {r=0.18}, BaseColor=@n115, Metallic=@n115, Specular=@n115, Roughness=@n115) {attributeSetTypes [
    #1
    #2
    #3
    #4
    #5
   ]}
  n117 = GetMaterialAttributes(None=FunctionInput(Preview=@n116) {inputName=MA, inputType=FunctionInput_MaterialAttributes}) {attributeGetTypes [#1]}
  NamedRerouteDeclaration(None=@n117.MaterialAttributes) {name=MA_In, nodeColor=<rgba 1 0.8672 0 1>}
  NamedRerouteDeclaration(None=@n117.EmissiveColor) {name=Emissive_In, nodeColor=<rgba 1 0.8672 0 1>}
  n118 = SmoothStep(
   Min=ScalarParameter() {defaultValue=0.5, group=01_HatchingLines, parameterName=Line_Thickness_Min, sliderMax=1}
   Max=ScalarParameter() {defaultValue=1, group=01_HatchingLines, parameterName=Line_Thickness_Max, sliderMax=1}
   Value=StaticSwitchParameter(True=NamedRerouteUsage(), False=NamedRerouteUsage()) {group=01_HatchingLines, parameterName=UseProceduralHachingLines, sortPriority=1}
  ) {constMin=0.5}
  n119 = TextureSampleParameter2D() {group=01_HatchingLines, parameterName=MacroVariationNoise, sortPriority=2, texture=/Engine/EngineMaterials/T_Default_MacroVariation}
  n120 = Multiply(A=@n119.R, B=ScalarParameter() {defaultValue=0.5, group=01_HatchingLines, parameterName=HatchingLines_NoiseIntensity})
  NamedRerouteDeclaration(None=Saturate(None=Add(A=@n118, B=@n120))) {name=HatchingLines, nodeColor=<rgba 0 0.9062 1 1>}
  NamedRerouteUsage()
  NamedRerouteUsage()
  MaterialExpressionReroute_4 = Reroute(None=@n6)
  n121 = Frac(None=@MaterialExpressionReroute_4.In)
  MaterialExpressionReroute_6 = Reroute(None=@n7)
  n122 = OneMinus(None=@MaterialExpressionReroute_6.RampWidth)
  n123 = Max(A=Divide(A=Subtract(A=@n121, B=@n122), B=@MaterialExpressionReroute_6.RampWidth), B=Constant() {r=0.5}) {desc=It is good to make functions look like a lancer}
  n124 = Add(A=Floor(None=@MaterialExpressionReroute_4.In), B=ConstantBiasScale(None=Cosine(None=@n123)))
  PinBase!(Output=@n124) {reroutePins [Output {expression=&self:MaterialExpressionReroute_7}]}
  PinBase!() {reroutePins [
    In {expression=&self:MaterialExpressionReroute_4}
    RampWidth {expression=&self:MaterialExpressionReroute_6}
   ]}
  PinBase() {reroutePins [
    In {expression=&self:MaterialExpressionReroute_0}
    RampWidth {expression=&self:MaterialExpressionReroute_1}
   ]}
  PinBase!(Output=@n46) {reroutePins [Output {expression=&self:MaterialExpressionReroute_5}]}
  Power(Base=@n109) {constExponent=32}
  MaterialExpressionReroute_7 = Reroute(None=@n124)
  n125 = Rotator(Coordinate=NamedRerouteUsage(), Time=Constant() {r=1}) {speed=0.5}
  TextureSample(UVs=Multiply(A=@n125) {constB=14}) {samplerType=SAMPLERTYPE_Masks, texture=&T_HatchingPattern04_FX}
  TextureSample(UVs=Multiply(A=MaterialFunctionCall(XY=@n125) {materialFunction=&Swizzle}.YX) {constB=18}) {samplerType=SAMPLERTYPE_Masks, texture=&T_HatchingPattern04_FX}
 ]}
