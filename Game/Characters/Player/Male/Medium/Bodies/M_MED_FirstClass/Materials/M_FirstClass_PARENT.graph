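# Parent material for the medium male player's FirstClass body, giving it anime shading and hatching lines on the Cloth shading model.
# Material attributes chain through shared character PreFX, FirstClass anime shading, hatching lines, a cloth-sim WPO function, then shared PostFX.
# Anime shading also passes an FX mask to the cloth function, and PreFX feeds AO straight to PostFX.
# Usage covers skeletal meshes, clothing, morph targets and mesh deformers.
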Material /Game/Characters/Player/Male/Medium/Bodies/M_MED_FirstClass/Materials/M_FirstClass_PARENT
graph {
 properties {
  bAutomaticallySetUsageInEditor=false
  bUseMaterialAttributes=true
  bUsedWithClothing=true
  bUsedWithMeshDeformer=true
  bUsedWithMorphTargets=true
  bUsedWithSkeletalMesh=true
  pixelDepthOffsetMode=PDOM_Legacy
  shadingModel=MSM_Cloth
  usedShadingModels=Cloth
 }
 nodes [
  n1 = MaterialFunctionCall() {materialFunction=&BaseFN_CharacterFunction_PreFX}
  n2 = MaterialFunctionCall!(Material=@n1.Result, CustomUVs) {materialFunction=&/MF_FirstClass_AnimeShading}
  MaterialFunctionCall(
   AO=@n1.AO
   Distance Blend=@n1.AO
   Fresnel=@n1.AO
   Gmap=@n1.AO
   Pre Skinned Local Position=@n1.AO
   Material=MaterialFunctionCall!(
    Material=MaterialFunctionCall!(MA=@n2.Result, SSC=@n2.Result, Color Mask=@n2.Result) {materialFunction=&/MF_FirstClass_HatchingLines}.Result
    SkinFX_Mask=@n2.FX
   ) {materialFunction=/Game/Global/MaterialFunctions/MF_ClothSimWPO_v2}.Result
  ) {materialFunction=&BaseFN_CharacterFunction_PostFX}
 ]
}
